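# Translucent unlit Niagara mesh-particle material for the NarroFleaMonkey air-wash trail in the Tails Spin effect.
# Slash texture sets opacity, eroded by a noise-and-UV-gradient dissolve driven by dynamic parameters, then scaled by particle alpha and depth fade.
# Slash texture also distorts panning directional noise, which blends with UV gradients between two colors for particle-tinted emissive.
# Refraction follows the dissolved mask and eases toward neutral with camera distance.
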Material /Game/Effects/Fort_Effects/Gameplay/Tails_Spin/M_NarroFleaMonkey_AirWash
graph {
 properties {
  bUsedWithNiagaraMeshParticles=true
  blendMode=BLEND_Translucent
  pixelDepthOffsetMode=PDOM_Legacy
  refractionMethod=RM_IndexOfRefraction
  shadingModel=MSM_Unlit
  usedShadingModels=Unlit
 }
 nodes [
  n1 = ScalarParameter() {defaultValue=1, parameterName=Ref}
  n2 = Distance(A=WorldPosition().XYZ, B=CameraPositionWS())
  n3 = Saturate(None=Divide(A=Subtract(A=@n2) {constB=350}) {constB=1600})
  n4 = ParticleColor()
  n5 = TextureObjectParameter() {
   group=SlashTexture
   parameterName=SlashTexture_Texture
   sortPriority=0
   texture=&/T_NarroFleaMonkey_AirTrail
  }
  n6 = VectorParameter() {
   group=SlashTexture
   parameterName=SlashTexture_UVControl
   channelNames {a=VPan, b=UPan, g=VScale, r=UScale}
   defaultValue=<rgba 1 2.5873 0 0>
  }
  n7 = LinearInterpolate(A=TextureCoordinate(), B=TextureCoordinate() {coordinateIndex=1}, Alpha=ScalarParameter() {parameterName=UseUV1})
  n8 = MaterialFunctionCall!(
   Texture=@n5
   AdditionalUVOffset
   UVControl_RGB=@n6
   UVControl_A=@n6
   "[Override]TimeInput"
   "[Override]ScaleUVByCenter"=StaticBool() {value=true}
   "[Override]UVInput"=@n7
  ) {materialFunction=/Game/Effects/Niagara/MasterMaterials/Functions/Common/FXCommon_TextureSample_V2}
  n9 = Multiply(
   A=Add(A=@n8.RGB, B=ScalarParameter() {group=Opacity, parameterName=Opacity_GrayAdjust_Add})
   B=ScalarParameter() {defaultValue=1, group=Opacity, parameterName=Opacity_GrayAdjust_Mul}
  )
  n10 = Saturate(None=Power(Base=Saturate(None=@n9), Exp=ScalarParameter() {defaultValue=1, group=Opacity, parameterName=Opacity_GrayAdjust_Power}))
  n11 = ScalarParameter() {defaultValue=2, parameterName=OpacityScale}
  n12 = Saturate(None=Multiply(
    A=MaterialFunctionCall!(TexCoord=@n8.UV, ClampChannel) {materialFunction=/Game/Effects/Fort_Effects/MaterialFunctions/MF_FX_TilingMethod}.Result
    B=Multiply(A=@n10, B=@n11)
   ))
  n13 = TextureObjectParameter() {
   group=DissolveA
   parameterName=DissolveA_Texture
   sortPriority=0
   texture=/Game/CharClasses/Outlander/Gadgets/AntiMaterialCharge/Materials/T_AntiMaterialCharge_Dissolve_01
  }
  n14 = VectorParameter() {group=DissolveA, parameterName=DissolveA_UVControl, channelNames {a=VPan, b=UPan, g=VScale, r=UScale}, defaultValue=<rgba 1 3 0.5 0>}
  n15 = OneMinus(None=MaterialFunctionCall(Float2=TextureCoordinate()) {materialFunction=&BreakOutFloat2Components}.R)
  n16 = LinearInterpolate(
   A=ScalarParameter() {defaultValue=-1.5, parameterName=DissolveA_BottomV}
   B=ScalarParameter() {defaultValue=0.28, parameterName=DissolveA_TopV}
   Alpha=Add(
    A=ScalarParameter() {parameterName=Dissolve}
    B=DynamicParameter() {defaultValue=<rgba 0 100 1 1>, paramNames [Dissolve, DepthFadeDistance, Param3, Param4]}.Dissolve
   )
  )
  n17 = Multiply(
   A=Add(
    A=MaterialFunctionCall(Texture=@n13, UVControl_RGB=@n14, UVControl_A=@n14) {materialFunction=&FXCommon_TextureSample}.R
    B=Add(A=@n15, B=@n16)
   )
   B=ScalarParameter() {defaultValue=1, group=DissolveA, parameterName=DissolveA_GrayAdjust_Mul}
  )
  n18 = LinearInterpolate(
   A=MaterialFunctionCall(In=Saturate(None=@n17), Contrast=ScalarParameter() {defaultValue=1, parameterName=C}) {materialFunction=&CheapContrast}.Result
   Alpha=ScalarParameter() {parameterName=DisableDissolve}
  ) {constB=0}
  n19 = Multiply(A=@n4.A, B=Saturate(None=Subtract(A=@n12, B=@n18)))
  LinearInterpolate(
   B=LinearInterpolate(B=LinearInterpolate(A=@n1, Alpha=@n3), Alpha=@n19) {constA=1}
   Alpha=ScalarParameter() {defaultValue=1, parameterName=UseRefraction}
  ) {constA=1, constB=0}
  n20 = TextureObjectParameter() {
   group=LDTexture
   parameterName=LDTexture_Texture
   samplerType=SAMPLERTYPE_LinearGrayscale
   sortPriority=0
   texture=/Game/Effects/Fort_Effects/Emotes/UltraLight/Texture/T_DirectionalNoise_Greyscale
  }
  n21 = VectorParameter() {
   group=SlashDistort
   parameterName=SlashDistort_NoiseControl
   sortPriority=60
   channelNames {a=VStrength, b=UStrength, g=VOffset, r=UOffset}
   defaultValue=<rgba -0.5 -0.5 0 0.1>
  }
  n22 = VectorParameter() {group=LDTexture, parameterName=LDTexture_UVControl, channelNames {a=VPan, b=UPan, g=VScale, r=UScale}, defaultValue=<rgba 1 2 0.5 0>}
  n23 = MaterialFunctionCall(
   Texture=@n20
   AdditionalUVOffset=MaterialFunctionCall!(NoiseValue=@n8.RG, Control_RGB=@n21, Control_A=@n21) {materialFunction=/Game/Effects/Niagara/MasterMaterials/Functions/Common/FXCommon_NoiseControl}.NoiseValue
   UVControl_RGB=@n22
   UVControl_A=@n22
   "[Override]UVInput"=@n7
  ) {materialFunction=&FXCommon_TextureSample}
  n24 = MaterialFunctionCall(Float2=TextureCoordinate()) {materialFunction=&BreakOutFloat2Components}
  n25 = Multiply(
   A=Add(A=@n24.R, B=ScalarParameter() {group=SlashFlowMask, parameterName=SlashFlowMask_GrayAdjust_Add})
   B=ScalarParameter() {defaultValue=1.2, group=SlashFlowMask, parameterName=SlashFlowMask_GrayAdjust_Mul}
  )
  n26 = Saturate(None=Power(Base=Saturate(None=@n25), Exp=ScalarParameter() {defaultValue=2, group=SlashFlowMask, parameterName=SlashFlowMask_GrayAdjust_Power}))
  n27 = Add(A=@n24.R, B=ScalarParameter() {group=SlashTexture, parameterName=SlashTexture_GrayAdjust_Add})
  n28 = ScalarParameter() {defaultValue=1, group=SlashTexture, parameterName=SlashTexture_GrayAdjust_Mul}
  n29 = Power(
   Base=Saturate(None=Multiply(A=@n27, B=@n28))
   Exp=ScalarParameter() {defaultValue=3, group=SlashTexture, parameterName=SlashTexture_GrayAdjust_Power}
  )
  n30 = LinearInterpolate(
   A=VectorParameter() {parameterName=Color2, defaultValue=<rgba 0.0886 0.1488 0.209 1>}.RGB
   B=VectorParameter() {parameterName=Color, defaultValue=<rgba 1 5.5 10 1>}.RGB
   Alpha=Multiply(A=LinearInterpolate(A=@n23.R, Alpha=@n26) {constA=1}, B=Saturate(None=@n29))
  )
  Multiply(
   A=MaterialFunctionCall(Emissive=Divide(A=@n30)) {materialFunction=&MF_FX_EmissiveAdjustments}.Result
   B=@n4.RGB
  )
  Multiply(
   A=@n19
   B=DepthFade(FadeDistance=DynamicParameter() {defaultValue=<rgba 0 100 1 1>, paramNames [Dissolve, DepthFadeDistance, Param3, Param4]}.DepthFadeDistance)
  )
 ]
}
